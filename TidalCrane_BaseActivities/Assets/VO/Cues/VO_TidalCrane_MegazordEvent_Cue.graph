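# Sound cue for Megazord event voice-over in the TidalCrane base activities content, routed through the narrator VO sound class.
# A Random node picks, with equal weight, one of three single-line waves, one each from the ClumsyChew, SwoopClasp and TalonPane sets.
# Playback uses the base activities VO attenuation and concurrency assets, which position the line and manage overlapping voice lines.
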SoundCue /TidalCrane_BaseActivities/Assets/VO/Cues/VO_TidalCrane_MegazordEvent_Cue
root=Random_0
nodes [
 {id=Random_0, children [WavePlayer_0,WavePlayer_1,WavePlayer_2], properties {weights [1,1,1]}}
 {id=WavePlayer_0, properties {soundWaveAssetPtr=/TidalCrane_BaseActivities/Assets/VO/Waves/ClumsyChew/VO_TidalCrane_ClumsyChew_MegaEvent_010}}
 {id=WavePlayer_1, properties {soundWaveAssetPtr=/TidalCrane_BaseActivities/Assets/VO/Waves/SwoopClasp/VO_TidalCrane_SwoopClasp_MegaEvent_010}}
 {id=WavePlayer_2, properties {soundWaveAssetPtr=/TidalCrane_BaseActivities/Assets/VO/Waves/TalonPane/VO_TidalCrane_TalonPane_MegaEvent_011}}
]
properties {
 attenuationSettings=/TidalCrane_BaseActivities/Assets/VO/AS_Fort_BaseActivities
 maxDistance=10100
 soundClassObject=/Game/Sounds/FortSoundClassesAndModes/Fort_VO_Narrator
 volumeMultiplier=1
 concurrencySet [/TidalCrane_BaseActivities/Assets/VO/BaseActivites_VO_Conc]
}
